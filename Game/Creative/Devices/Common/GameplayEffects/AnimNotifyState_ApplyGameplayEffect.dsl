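; Anim notify state in Creative devices common content that keeps a gameplay effect on a Fortnite player pawn for the notify window.
; On begin, it casts the mesh owner to a player pawn and self-applies the configured effect class through its ability system component.
; On end, it removes that effect by class rather than by stored handle, clearing all stacks, including any applied from elsewhere.
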(fn Received_NotifyBegin
  (params (MeshComp :type Skeletal%20Mesh%20Component%20Object%20Reference) (Animation :type Anim%20Sequence%20Base%20Object%20Reference) (TotalDuration :type Float%20%28single-precision%29) (EventReference :type Anim%20Notify%20Event%20Reference%20Structure%20%28by%20ref%29))
  (bind _asfort_player_pawn (Utilities|Casting|CastToFortPlayerPawn (Components|GetOwner MeshComp))
    (:then
      (Utilities|IsValid (Ability|GetAbilitySystemComponent _asfort_player_pawn)
        (:Is%20Valid
          (bind _appliedgameplayeffect (Variables|Default|GetAppliedGameplayEffect)
            (:then
              (GameplayEffects|ApplyGameplayEffectToSelf (Ability|GetAbilitySystemComponent _asfort_player_pawn) _appliedgameplayeffect 0.0 (GameplayEffects|MakeEffectContext (Ability|GetAbilitySystemComponent _asfort_player_pawn)))
              (return true))
            (:else)))
        (:Is%20Not%20Valid
          _)))
    (:CastFailed
      _))
  (return false))

(fn Received_NotifyEnd
  (params (MeshComp :type Skeletal%20Mesh%20Component%20Object%20Reference) (Animation :type Anim%20Sequence%20Base%20Object%20Reference) (EventReference :type Anim%20Notify%20Event%20Reference%20Structure%20%28by%20ref%29))
  (bind _asfort_player_pawn (Utilities|Casting|CastToFortPlayerPawn (Components|GetOwner MeshComp))
    (:then
      (Utilities|IsValid (Ability|GetAbilitySystemComponent _asfort_player_pawn)
        (:Is%20Valid
          (bind _appliedgameplayeffect (Variables|Default|GetAppliedGameplayEffect)
            (:then
              (GameplayEffects|RemoveActiveGameplayEffectbySourceEffect (Ability|GetAbilitySystemComponent _asfort_player_pawn) _appliedgameplayeffect 0 -1)
              (return true))
            (:else)))))
    (:CastFailed
      (return false))))
